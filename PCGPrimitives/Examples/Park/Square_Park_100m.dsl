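; PCG example graph that builds a square park from a created spline shape, composed of PCGPrimitives subgraphs, and outputs that shape.
; Recursive splitting divides it into areas and split splines; the latter gain shape-grammar footprints, whose expanded areas cut the boundary segments before rule assignment.
; An offset copy of the shape becomes planar mesh ground.
; Contracted areas pass a random choice, then get randomly filled with spawned assets.
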(graph Square_Park_100m
  (bind Subgraph_36 (Subgraph :Graph /PCGPrimitives/Primitives/Create/Create_Spline_Shape.Create_Spline_Shape))
  (bind Subgraph_2 (Subgraph :Graph /PCGPrimitives/Primitives/Subdivide/Subdivide_Areas_WithRecursiveSplit.Subdivide_Areas_WithRecursiveSplit))
  (bind Subgraph_4 (Subgraph :Graph /PCGPrimitives/Primitives/Assign/Assign_ShapeGrammarDefinition.Assign_ShapeGrammarDefinition))
  (bind Subgraph_0 (Subgraph :Graph /PCGPrimitives/Primitives/Extract/Extract_Segments.Extract_Segments))
  (bind Subgraph_13 (Subgraph :Graph /PCGPrimitives/Primitives/Transform/Transform_Areas_Expand.Transform_Areas_Expand))
  (bind Subgraph_11 (Subgraph :Graph /PCGPrimitives/Primitives/Compose/Compose_Spline_Cut.Compose_Spline_Cut))
  (bind Subgraph_37 (Subgraph :Graph /PCGPrimitives/Primitives/Extract/Extract_Segments.Extract_Segments))
  (bind Subgraph_8 (Subgraph :Graph /PCGPrimitives/Primitives/Assign/Assign_Rule.Assign_Rule))
  (bind Subgraph_10 (Subgraph :Graph /PCGPrimitives/Primitives/Transform/Transform_Offset.Transform_Offset))
  (bind Subgraph_9 (Subgraph :Graph /PCGPrimitives/Primitives/Create/Create_Mesh_Planar.Create_Mesh_Planar))
  (bind Subgraph_14 (Subgraph :Graph /PCGPrimitives/Primitives/Transform/Transform_Areas_Contract.Transform_Areas_Contract))
  (bind Subgraph_15 (Subgraph :Graph /PCGPrimitives/Primitives/Filter/Filter_RandomChoice.Filter_RandomChoice))
  (bind Subgraph_16 (Subgraph :Graph /PCGPrimitives/Primitives/Fill/Fill_Areas_Randomly.Fill_Areas_Randomly))
  (bind Subgraph_20 (Subgraph :Graph /PCGPrimitives/Primitives/Spawn/Spawn_Assets.Spawn_Assets))
  (connect Subgraph_36 Out Subgraph_0 PrimaryInput)
  (connect Subgraph_36 Out Subgraph_2 PrimaryInput)
  (connect Subgraph_2 Subdivide%20Splines Subgraph_4 PrimaryInput)
  (connect Subgraph_37 Out Subgraph_8 PrimaryInput)
  (connect Subgraph_10 Out Subgraph_9 PrimaryInput)
  (connect Subgraph_36 Out Subgraph_10 PrimaryInput)
  (connect Subgraph_0 Out Subgraph_11 PrimaryInput)
  (connect Subgraph_13 Out Subgraph_11 SecondaryInput)
  (connect Subgraph_4 Footprints Subgraph_13 PrimaryInput)
  (connect Subgraph_2 Areas Subgraph_14 PrimaryInput)
  (connect Subgraph_14 Out Subgraph_15 PrimaryInput)
  (connect Subgraph_15 Selected Subgraph_16 PrimaryInput)
  (connect Subgraph_16 Out Subgraph_20 PrimaryInput)
  (connect Subgraph_11 Out Subgraph_37 PrimaryInput)
  (output Out Subgraph_36))
